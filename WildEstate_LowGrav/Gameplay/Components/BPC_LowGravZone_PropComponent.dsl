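; Component that makes a BuildingProp float in Wild Estate low-gravity zones and react to hits.
; Once playlist data is ready, it binds damage and tagged-collider overlap events, applying low gravity immediately if enabled.
; A physics-object gravity override and attached VFX toggle on overlaps; damage and pawn or vehicle contact add server-side impulses.
; Flags optionally destroy the prop on vehicle collision or zone exit, and contacts also trigger collision audio.
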(event EventBeginPlay
  (|CallParentFunction)
  (Class|FortAsyncActionCurrentPlaylistReady|PlaylistDataReadyAsync
    (:OnReady
      (Default|InitializeOwningBuildingProp)
      ;[0 0 0 0.5] Enable low gravity effect immediately if the effect should be active
      (if (ScalableFloat|GetValueasBool (Variables|Default|GetLowGravEnabled))
        (CallFunction|EnableLowGravityEffect)))))

(event EventEndPlay
  (params (EndPlayReason :type EEndPlayReason%20Enum :default "Destroyed"))
  (|CallParentFunction EndPlayReason)
  (CallFunction|UnbindEvents))

(fn AddImpulse
  (params (HitVelocity :type Vector) (VelocityMultiplier :type Float%20%28double-precision%29) (HItLocation :type Vector) (Instigator :type Actor%20Object%20Reference))
  (FortPhysicsSimulationLibrary|AddPhysicsImpulseatLocation (Components|GetOwner) (* (Variables|GetHitVelocity) (Variables|GetVelocityMultiplier)) (Variables|GetHItLocation) true "None" false (Variables|GetInstigator) (struct :MaxSimSizeAllowed "Medium" :bRejectEmbeddedInLandscape true)))

(fn DisableLowGravityEffect
  (bind _owningbuildingprop (Variables|Default|GetOwningBuildingProp)
    (:then
      (bind _returnvalue (BuildingActor|PhysicsSimulationSupport|GetPhysicsObjectComponent _owningbuildingprop))
      (PhysicsObject|SetGravityOverride _returnvalue false 0.0)
      (Networking|CanExecuteCosmeticEvents
        (:True
          (bind _lowgravityfx (Variables|Default|GetLowGravityFX)
            (:then
              (Components|Activation|Deactivate _lowgravityfx))
            (:else)))))
    (:else)))

(fn EnableLowGravityEffect
  ;[0 0 0 0.5] Override gravity magnitude when entering the low gravity zone
  (bind _owningbuildingprop (Variables|Default|GetOwningBuildingProp)
    (:then
      (bind _returnvalue (BuildingActor|PhysicsSimulationSupport|GetPhysicsObjectComponent _owningbuildingprop))
      (PhysicsObject|WakeUp _returnvalue "AlwaysInstant" :!class /Script/FortniteGame.FortPhysicsObjectComponent)
      (PhysicsObject|SetGravityOverride _returnvalue true (ScalableFloat|GetValueasDouble (Variables|Default|GetPropGravityOverride)))
      (Networking|CanExecuteCosmeticEvents
        (:True
          (Default|SpawnandActivateLowGravVFX))))
    (:else)))

(fn GetLowGravCollisionComponent
  (bind _lowgravcollidertag (Variables|Default|GetLowGravColliderTag))
  (Utilities|FlowControl|Sequence
    (:then_0
      (bind _owningbuildingprop (Variables|Default|GetOwningBuildingProp)
        (:then
          (if (Utilities|Array|IsNotEmpty (Actor|GetComponentsbyTag _owningbuildingprop "/Script/Engine.ShapeComponent" _lowgravcollidertag))
            (return (Utilities|Array|Get(acopy) (Actor|GetComponentsbyTag _owningbuildingprop "/Script/Engine.ShapeComponent" _lowgravcollidertag)))))
        (:else)))
    (:then_1
      (return 0))))

(fn GetOwningBuildingProp
  (bind _asbuilding_prop (Utilities|Casting|CastToBuildingProp (Components|GetOwner))
    (:then
      (return _asbuilding_prop))
    (:CastFailed
      (return 0))))

(fn InitializeOwningBuildingProp
  (bind _buildingactor (CallFunction|GetOwningBuildingProp))
  (bind _shapecomponent (CallFunction|GetLowGravCollisionComponent))
  (Utilities|IsValid _buildingactor
    (:Is%20Valid
      (Variables|Default|SetOwningBuildingProp _buildingactor)
      (Variables|/Script/FortniteGame.BuildingActor|SetbIsInvulnerable (Variables|Default|GetbBuildingPropIsInvulnerable) (Variables|Default|GetOwningBuildingProp))
      ;[0 0 0 0.5] Hook up BuildingProp damaged event
      (EventDispatchers|BindEventtoOnDamaged (Variables|Default|GetOwningBuildingProp) (EventDispatchers|CreateEvent :!event OnBuildingPropDamaged))
      ;[0 0 0 0.5] Initialize shape component
      (Utilities|IsValid _shapecomponent
        (:Is%20Valid
          (bind _output_get (Variables|Default|SetLowGravCollisionComponent _shapecomponent))
          (Collision|BindEventtoOnComponentBeginOverlap _output_get (EventDispatchers|CreateEvent :!event OnColliderBeginOverlap))
          (Collision|BindEventtoOnComponentEndOverlap _output_get (EventDispatchers|CreateEvent :!event OnColliderEndOverlap))
          ;[0 0 0 0.5] Cache collision audio component
          (Variables|Default|SetFortCollisionAudioComponent (Actor|GetComponentbyClass (Variables|Default|GetOwningBuildingProp) "/Script/FortniteGame.FortCollisionAudioComponent")))
        (:Is%20Not%20Valid
          (Fortnite|FortShippingLog self "Failed to find Low Grav Collision Component. Ensure the LowGravColliderTag exists on the Owning Building Prop actor's collision component." true))))))

(fn OnBuildingPropDamaged
  (params (DamagedActor :type Actor%20Object%20Reference) (Damage :type Float%20%28single-precision%29) (InstigatedBy :type Controller%20Object%20Reference) (DamageCauser :type Actor%20Object%20Reference) (HitLocation :type Vector) (FHitComponent :type Primitive%20Component%20Object%20Reference) (BoneName :type Name) (Momentum :type Vector))
  ;[0 0 0 0.5] Add additional impulse if gravity effect is enabled
  (if (Networking|IsServer)
    (if (ScalableFloat|GetValueasBool (Variables|Default|GetLowGravEnabled))
      (CallFunction|AddImpulse (Variables|GetMomentum) (ScalableFloat|GetValueasDouble (Variables|Default|GetPropDamageImpulseMultiplier)) (Variables|GetHitLocation) (Variables|GetInstigatedBy)))))

(fn OnColliderBeginOverlap
  (params (OverlappedComponent :type Primitive%20Component%20Object%20Reference) (OtherActor :type Actor%20Object%20Reference) (OtherComp :type Primitive%20Component%20Object%20Reference) (OtherBodyIndex :type Integer) (bFromSweep :type Boolean) (SweepResult :type Hit%20Result%20Structure%20%28by%20ref%29))
  (bind (_bblockinghit _binitialoverlap _time _distance _location) (Collision|BreakHitResult (Variables|GetSweepResult)))
  (bind (_bblockinghit_2 _binitialoverlap_2 _time_2 _distance_2 _location_2 _impactpoint _normal _impactnormal) (Collision|BreakHitResult (Variables|GetSweepResult)))
  (bind (_bblockinghit_3 _binitialoverlap_3 _time_3 _distance_3 _location_3 _impactpoint_2 _normal_2 _impactnormal_2 _physmat _hitactor) (Collision|BreakHitResult (Variables|GetSweepResult)))
  (if (ScalableFloat|GetValueasBool (Variables|Default|GetLowGravEnabled))
    (Utilities|FlowControl|Sequence
      (:then_0
        ;[0 0 0 0.5] Enable low gravity effects when entering the low gravity zone
        (Utilities|Casting|CastToA_WildEstate_LowGravZone (Variables|GetOtherActor)
          (:then
            (CallFunction|EnableLowGravityEffect))
          (:CastFailed
            ;[0 0 0 0.5] Cache relevant overlapping actor
            (bind _asfort_pawn (Utilities|Casting|CastToFortPawn (Variables|GetOtherActor))
              (:then
                (LocalVariables|SetRelevantActor _asfort_pawn))
              (:CastFailed
                (bind _asfort_athena_vehicle (Utilities|Casting|CastToFortAthenaVehicle (Variables|GetOtherActor)))
                (if (Variables|Default|GetbDestroyOnVehicleCollision)
                  (bind _owningbuildingprop (Variables|Default|GetOwningBuildingProp)
                    (:then
                      (Gameplay|Die _owningbuildingprop "(TagName=\"Gameplay.Damage.Vehicle\")" (Fortnite|GetControllerfromActor _asfort_athena_vehicle) _asfort_athena_vehicle :!class /Script/FortniteGame.BuildingActor))
                    (:else))
                  (else
                    (LocalVariables|SetRelevantActor _asfort_athena_vehicle))))))))
      (:then_1
        ;[0 0 0 0.5] Add impulse if overlapping with a FortPawn or FortAthenaVehicle
        (if (Networking|IsServer)
          (bind _relevantactor (LocalVariables|GetRelevantActor)
            (:then
              (Utilities|IsValid _hitactor
                (:Is%20Valid
                  (CallFunction|AddImpulse (Transformation|GetVelocity _relevantactor) (ScalableFloat|GetValueasDouble (Variables|Default|GetPropCollisionImpulseMultiplier)) _location (LocalVariables|GetRelevantActor)))
                (:Is%20Not%20Valid
                  (Fortnite|FortShippingLog self "OnColliderBeginOverlap Sweep Result returned an invalid hit actor" true))))
            (:else))))
      (:then_2
        ;[0 0 0 0.5] Play collision audio
        (bind _fortcollisionaudiocomponent (Variables|Default|GetFortCollisionAudioComponent)
          (:then
            (Audio|OnCollision _fortcollisionaudiocomponent _location_2 _impactnormal))
          (:else))))))

(fn OnColliderEndOverlap
  (params (OverlappedComponent :type Primitive%20Component%20Object%20Reference) (OtherActor :type Actor%20Object%20Reference) (OtherComp :type Primitive%20Component%20Object%20Reference) (OtherBodyIndex :type Integer))
  ;[0 0 0 0.5] Stop overriding gravity magnitude when exiting the low gravity zone
  (if (ScalableFloat|GetValueasBool (Variables|Default|GetLowGravEnabled))
    (bind _asa_wild_estate_low_grav_zone (Utilities|Casting|CastToA_WildEstate_LowGravZone OtherActor))
    (CallFunction|DisableLowGravityEffect)
    ;[0 0 0 0.5] Conditionally destroy the building prop if it has exited the LowGravZone collider
    (if (Variables|Default|GetbDestroyOnExitLowGravZone)
      (bind _owningbuildingprop (Variables|Default|GetOwningBuildingProp)
        (:then
          (Gameplay|Die _owningbuildingprop "(TagName=\"Gameplay.Damage.Environment\")" 0 _asa_wild_estate_low_grav_zone :!class /Script/FortniteGame.BuildingActor))
        (:else)))))

(fn SpawnAndActivateLowGravVFX
  (bind _owningbuildingprop (Variables|Default|GetOwningBuildingProp)
    (:then
      (bind _lowgravityfx (Variables|Default|GetLowGravityFX)
        (:then
          (Components|Activation|Activate _lowgravityfx false))
        (:else
          (bind _returnvalue (Effects|SpawnFXSystemAttached (Variables|Default|GetLowGravFXSystem%20Asset) (Variables|/Script/Engine.Actor|GetRootComponent _owningbuildingprop) "None" (Variables|Default|GetLowGravityFXOffset)))
          (Variables|Default|SetLowGravityFX _returnvalue))))
    (:else)))

(fn UnbindEvents
  (Utilities|FlowControl|Sequence
    (:then_0
      (bind _lowgravityfx (Variables|Default|GetLowGravityFX)
        (:then
          (Components|DestroyComponent _lowgravityfx)
          (Variables|Default|SetLowGravityFX 0))
        (:else)))
    (:then_1
      (bind _lowgravcollisioncomponent (Variables|Default|GetLowGravCollisionComponent)
        (:then
          (Collision|UnbindEventfromOnComponentBeginOverlap _lowgravcollisioncomponent (EventDispatchers|CreateEvent :!event OnColliderBeginOverlap))
          (Collision|UnbindEventfromOnComponentEndOverlap _lowgravcollisioncomponent (EventDispatchers|CreateEvent :!event OnColliderEndOverlap))
          (Variables|Default|SetLowGravCollisionComponent 0))
        (:else)))
    (:then_2
      (bind _owningbuildingprop (Variables|Default|GetOwningBuildingProp)
        (:then
          (EventDispatchers|UnbindEventfromOnDamaged _owningbuildingprop (EventDispatchers|CreateEvent :!event OnBuildingPropDamaged))
          (Variables|Default|SetOwningBuildingProp 0))
        (:else)))))
